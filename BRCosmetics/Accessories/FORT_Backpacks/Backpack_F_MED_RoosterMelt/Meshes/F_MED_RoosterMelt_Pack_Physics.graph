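# Physics asset for the RoosterMelt back bling pack mesh, simulating a small dangling figure with body, head, arm and leg chains.
# Capsules cover the pack, torso, head, legs and upper arms, with spheres on the arm ends.
# The top torso bone hangs from the pack on a joint with free swing and twist.
# Most other joints have tight angular limits, while shoulders swing widely, the left more than the right.
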PhysicsAsset /BRCosmetics/Accessories/FORT_Backpacks/Backpack_F_MED_RoosterMelt/Meshes/F_MED_RoosterMelt_Pack_Physics
bodies {
 dyn_arm_end_L [{shapeType=Sphere, radius=1.25, shapeName=dyn_arm_end_L_sphere, center=<0.5016 0 0>}]
 dyn_arm_end_R [{shapeType=Sphere, radius=1.25, shapeName=dyn_arm_end_L_sphere, center=<-0.5016 0 0>}]
 dyn_arm_start_L [{
   shapeType=Capsule
   length=1.75
   radius=1.75
   shapeName=dyn_arm_start_L_capsule
   center=<2.2489 0.0955 -0.1353>
   rotation=<rot 89.1308 87.5456 87.4697>
  }]
 dyn_arm_start_R [{
   shapeType=Capsule
   length=1.75
   radius=1.75
   shapeName=dyn_arm_start_L_capsule
   center=<1.4326 -0.0906 0.1328>
   rotation=<rot 89.1308 87.5456 -92.5303>
  }]
 dyn_body_01 [{shapeType=Capsule, length=2.25, radius=1.25, shapeName=dyn_body_01_capsule, rotation=<rot 0.0001 0 -90>}]
 dyn_body_02 [{shapeType=Capsule, length=1, radius=2.25, shapeName=dyn_body_02_capsule, center=<2.0984 0 0>, rotation=<rot -90 179.9209 -179.9209>}]
 dyn_body_03 [{shapeType=Capsule, length=0.75, radius=2.5, shapeName=dyn_body_03_capsule, rotation=<rot -90 2.5235 -2.5235>}]
 dyn_head_end [{shapeType=Capsule, length=1.5, radius=1.25, shapeName=dyn_head_end_capsule, center=<-0.328 0 0.2614>, rotation=<rot 0.0001 0 90>}]
 dyn_head_start [{
   shapeType=Capsule
   length=1.5
   radius=1.75
   shapeName=dyn_head_start_capsule
   center=<-2.2117 0.121 0>
   rotation=<rot 90 -177.4765 -176.7396>
  }]
 dyn_leg_end_L [{
   shapeType=Capsule
   length=0.4204
   radius=1
   shapeName=dyn_leg_end_L_capsule
   center=<0.4122 0.0587 -0.0482>
   rotation=<rot 87.7392 120.7027 -149.6104>
  }]
 dyn_leg_end_R [{
   shapeType=Capsule
   length=0.4204
   radius=1
   shapeName=dyn_leg_end_L_capsule
   center=<0.5905 0.0333 0.0653>
   rotation=<rot 87.7392 120.7027 30.3896>
  }]
 dyn_leg_mid_L [{
   shapeType=Capsule
   length=1.7994
   radius=1.0155
   shapeName=dyn_leg_mid_L_capsule
   center=<2.1244 -0.1796 0.2525>
   rotation=<rot -0.6233 87.3836 -87.5847>
  }]
 dyn_leg_mid_R [{
   shapeType=Capsule
   length=1.5494
   radius=1.0155
   shapeName=dyn_leg_mid_L_capsule
   center=<2.2472 0.2102 -0.0708>
   rotation=<rot -0.8517 92.877 92.3446>
  }]
 dyn_leg_start_L [{
   shapeType=Capsule
   length=4.0368
   radius=0.9666
   shapeName=dyn_leg_start_R_capsule
   center=<1.9056 0.0144 0.1495>
   rotation=<rot -0.2893 89.566 93.3086>
  }]
 dyn_leg_start_R [{
   shapeType=Capsule
   length=3.7868
   radius=1.2166
   shapeName=dyn_leg_start_R_capsule
   center=<2.0986 -0.1269 0.0824>
   rotation=<rot -0.3839 91.2052 -86.7009>
  }]
 pack [{shapeType=Capsule, length=0.505, radius=0.505, shapeName=pack_capsule, rotation=<rot 0 -90 0>}]
}
constraints [
 {bone1Name=dyn_arm_start_R, bone2Name=dyn_body_03, swing1LimitDegrees=75, swing2LimitDegrees=12, twistLimitDegrees=6}
 {bone1Name=dyn_arm_end_R, bone2Name=dyn_arm_start_R, swing1LimitDegrees=6, swing2LimitDegrees=6, twistLimitDegrees=6}
 {bone1Name=dyn_head_start, bone2Name=dyn_body_03, swing1LimitDegrees=8, swing2LimitDegrees=8, twistLimitDegrees=1}
 {bone1Name=dyn_head_end, bone2Name=dyn_head_start, swing1LimitDegrees=4, swing2LimitDegrees=4, twistLimitDegrees=1}
 {bone1Name=dyn_body_02, bone2Name=dyn_body_03, swing1LimitDegrees=3, swing2LimitDegrees=3, twistLimitDegrees=2}
 {bone1Name=dyn_body_01, bone2Name=dyn_body_02, swing1LimitDegrees=2, swing2LimitDegrees=0, twistLimitDegrees=0}
 {bone1Name=dyn_leg_mid_L, bone2Name=dyn_leg_start_L, swing1LimitDegrees=4, swing2LimitDegrees=4, twistLimitDegrees=0}
 {bone1Name=dyn_leg_end_L, bone2Name=dyn_leg_mid_L, swing1LimitDegrees=9, swing2LimitDegrees=9, twistLimitDegrees=5}
 {bone1Name=dyn_leg_start_R, bone2Name=dyn_body_01, swing1LimitDegrees=11, swing2LimitDegrees=18, twistLimitDegrees=4}
 {bone1Name=dyn_leg_start_L, bone2Name=dyn_body_01, swing1LimitDegrees=16, swing2LimitDegrees=20, twistLimitDegrees=4}
 {bone1Name=dyn_leg_mid_R, bone2Name=dyn_leg_start_R, swing1LimitDegrees=4, swing2LimitDegrees=4, twistLimitDegrees=0}
 {bone1Name=dyn_leg_end_R, bone2Name=dyn_leg_mid_R, swing1LimitDegrees=9, swing2LimitDegrees=9, twistLimitDegrees=5}
 {bone1Name=dyn_arm_end_L, bone2Name=dyn_arm_start_L, swing1LimitDegrees=6, swing2LimitDegrees=6, twistLimitDegrees=6}
 {bone1Name=dyn_arm_start_L, bone2Name=dyn_body_03, swing1LimitDegrees=85, swing2LimitDegrees=45, twistLimitDegrees=0}
 {
  bone1Name=dyn_body_03
  bone2Name=pack
  swing1LimitDegrees=45
  swing1Motion=Free
  swing2LimitDegrees=45
  swing2Motion=Free
  twistLimitDegrees=45
  twistMotion=Free
 }
]
