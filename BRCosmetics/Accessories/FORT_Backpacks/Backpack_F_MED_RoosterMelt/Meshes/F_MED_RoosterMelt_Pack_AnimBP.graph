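# Animation Blueprint for the RoosterMelt backpack cosmetic, adding secondary motion to its body, flag and nail bones.
# Small root and dyn_body_03 offsets feed a RigidBody simulation in pack-bone space using the pack's physics asset.
# Chained spring-driven AnimDynamics nodes then move the flag start, wind-enabled flag end and partially weighted nail.
# Each update sets a gravity override through a deprecated skydive-aware function, though the AnimGraph shows fixed gravity values instead.
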AnimBlueprint /BRCosmetics/Accessories/FORT_Backpacks/Backpack_F_MED_RoosterMelt/Meshes/F_MED_RoosterMelt_Pack_AnimBP
target_skeleton=&/F_MED_RoosterMelt_Pack_Skeleton
graphs [
 AnimGraph {nodes [
   n1 = "Transform (Modify) Bone - Bone: root"(ComponentPose=Local To Component(LocalPose=Input Pose()), Translation="0,-1,0", Rotation="0,0,0") {
    translationMode=BMM_Additive
    translationSpace=BCS_BoneSpace
    alphaBoolBlend {blendOption=Linear}
    boneToModify {boneName=root}
    translation {y=-1}
   }
   n2 = "Transform (Modify) Bone - Bone: dyn_body_03"(ComponentPose=@n1, Alpha=1, Translation="0.000,0.000,2.057", Rotation="0,0,0", Scale="1,1,1") {
    translationMode=BMM_Additive
    translationSpace=BCS_BoneSpace
    alphaBoolBlend {blendOption=Linear}
    boneToModify {boneName=dyn_body_03}
    translation {z=2.057}
   }
   n3 = RigidBody!(ComponentPose=@n2, Alpha, OverrideWorldGravity:Vector="0,0,-980", ExternalForce) -> (Pose) {
    bClampLinearTranslationLimitToRefPose=true
    bOverrideWorldGravity=true
    overridePhysicsAsset=&/F_MED_RoosterMelt_Pack_Physics
    simulationSpace=BaseBoneSpace
    alphaBoolBlend {blendOption=Linear}
    baseBoneRef {boneName=pack}
    componentLinearAccScale=<0.2 0.2 0.2>
    componentLinearVelScale=<1 1 0.8>
    simSpaceSettings {maxAngularAcceleration=500, maxAngularVelocity=1, maxLinearAcceleration=400, maxLinearVelocity=250, velocityScaleZ=0, worldAlpha=0.4}
   }
   n4 = Anim Dynamics - Bone: dyn_flag_start(ComponentPose=@n3, Alpha=1, ExternalForce="0,0,0") {
    angularBiasOverride=1
    angularDampingOverride=1
    angularSpringConstant=20
    bAngularSpring=true
    bGravityOverrideInSimSpace=true
    bOverrideAngularDamping=true
    bUseGravityOverride=true
    simulationSpace=RootRelative
    alphaBoolBlend {blendOption=Linear}
    boundBone {boneName=dyn_flag_start}
    componentAppliedLinearAccClamp=<1000 1000 1000>
    componentLinearAccScale=<0.7 0.7 0.7>
    componentLinearVelScale=<0.4 0.4 0.4>
    gravityOverride {z=-920}
    physicsBodyDefinitions [{
      boundBone {boneName=dyn_flag_start}
      boxExtents=<2 2 2>
      constraintSetup {angularTargetAxis=AxisY, angularLimitsMax=<10 10 6>, angularLimitsMin=<-10 -10 -7>, angularTarget=<0 1 0>}
      localJointOffset=<0.1 0 0>
     }]
   }
   n5 = Anim Dynamics - Bone: dyn_flag_end!(ComponentPose=@n4, Alpha=1, GravityOverride:Vector="0,0,-920", ExternalForce="0,0,0") -> (Pose) {
    angularBiasOverride=1
    angularDampingOverride=0.9999
    angularSpringConstant=100
    bAngularSpring=true
    bEnableWind=true
    bGravityOverrideInSimSpace=true
    bOverrideAngularDamping=true
    bUseGravityOverride=true
    bUsePlanarLimit=false
    simulationSpace=RootRelative
    windScale=60
    alphaBoolBlend {blendOption=Linear}
    boundBone {boneName=dyn_flag_end}
    componentAppliedLinearAccClamp=<1000 1000 1000>
    componentLinearAccScale=<0.6 0.6 0.6>
    componentLinearVelScale=<0.4 0.4 0.4>
    gravityOverride {z=-920}
    physicsBodyDefinitions [{
      boundBone {boneName=dyn_flag_end}
      boxExtents=<2 1 1>
      constraintSetup {coneAngle=1, angularLimitsMax=<10 15 9>, angularLimitsMin=<-10 -10 -5>, angularTarget=<1 0 0>}
      localJointOffset=<0 0 0>
     }]
   }
   n6 = Anim Dynamics - Bone: dyn_nail(ComponentPose=@n5, Alpha=0.6, ExternalForce="0,0,0") {
    angularBiasOverride=1
    angularDampingOverride=1
    angularSpringConstant=30
    bAngularSpring=true
    bGravityOverrideInSimSpace=true
    bOverrideAngularDamping=true
    bUseGravityOverride=true
    simulationSpace=RootRelative
    alphaBoolBlend {blendOption=Linear}
    boundBone {boneName=dyn_nail}
    componentAppliedLinearAccClamp=<1000 1000 1000>
    componentLinearAccScale=<1 1 1>
    componentLinearVelScale=<1 1 1>
    gravityOverride {z=-920}
    physicsBodyDefinitions [{
      boundBone {boneName=dyn_nail}
      boxExtents=<2 2 2>
      constraintSetup {angularLimitsMax=<5 5 0>, angularLimitsMin=<-5 -4 0>, angularTarget=<1 0 0>}
      localJointOffset=<0.1 0 0>
     }]
   }
   Output Pose(Result=Component To Local(ComponentPose=@n6)) {layerGroup=None}
  ]}
 EventGraph {nodes [
   Event BlueprintUpdateAnimation()
   > n1 = Gravity Override State Function DEPRECATED(CharacterPartAnimInstance=Self-Reference(), GameTimeInSeconds=GetGameTimeInSeconds(), GravityOverrideParams=Get Gravity Override Params())
   > Set Gravity Overide(Gravity Overide:Vector=@n1) -> (Output_Get:Vector)
   TryGetPawnOwner()
  ]}
]
variables [
 Gravity Overide {type=Vector}
 Gravity Override Params {
  type=GravityOverrideParamsStruct
  default {skydive_FloatingDown {gravityOverride=<0 -700 -600>}, skydive_GoingUp {gravityOverride=<0 -300 2000>}}
 }
]
